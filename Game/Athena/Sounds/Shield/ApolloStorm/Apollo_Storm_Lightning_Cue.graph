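# Sound cue for Apollo storm lightning strikes, using safe-zone lightning attenuation and the storm sound class.
# It mixes pitch-modulated random close lightning, a louder heavy-lightning layer, and distant thunder delayed by one to five seconds.
# Null entries leave the heavy layer silent about half the time.
# Quality-level nodes switch the lower two tiers of each layer to smaller random pools that are excluded from branch culling.
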SoundCue /Game/Athena/Sounds/Shield/ApolloStorm/Apollo_Storm_Lightning_Cue
root=Mixer_0
nodes [
 {id=Delay_0, children [QualityLevel_0], properties {delayMax=5, delayMin=1}}
 {id=Mixer_0, children [Mixer_1,Delay_0], properties {inputVolume [1,1]}}
 {id=Mixer_1, children [Modulator_0,Modulator_1], properties {inputVolume [1,1.8]}}
 {id=Modulator_0, children [QualityLevel_1], properties {pitchMin=0.8, volumeMax=1, volumeMin=0.75}}
 {id=Modulator_1, children [QualityLevel_2], properties {pitchMax=1.2, pitchMin=0.8}}
 {id=QualityLevel_0, children [Random_3,Random_3,Random_0,Random_0]}
 {id=QualityLevel_1, children [Random_1,Random_1,Random_2,Random_2]}
 {id=QualityLevel_2, children [Random_4,Random_4,Random_5,Random_5]}
 {id=Random_0, children [WavePlayer_33, WavePlayer_34, WavePlayer_35], properties {bShouldExcludeFromBranchCulling=true, weights [1,1,1]}}
 {
  id=Random_1
  children [
   WavePlayer_12
   WavePlayer_13
   WavePlayer_14
   WavePlayer_15
   WavePlayer_16
   WavePlayer_17
   WavePlayer_18
   WavePlayer_19
   WavePlayer_20
   WavePlayer_23
  ]
  properties {weights [1,1,1,1,1,1,1,1,1,1]}
 }
 {id=Random_2, children [WavePlayer_19, WavePlayer_20, WavePlayer_23], properties {bShouldExcludeFromBranchCulling=true, weights [1,1,1]}}
 {
  id=Random_3
  children [
   WavePlayer_25
   WavePlayer_26
   WavePlayer_28
   WavePlayer_29
   WavePlayer_30
   WavePlayer_31
   WavePlayer_32
   WavePlayer_33
   WavePlayer_34
   WavePlayer_35
  ]
  properties {weights [1,1,1,1,1,1,1,1,1,1]}
 }
 {
  id=Random_4
  children [
   WavePlayer_0
   WavePlayer_1
   WavePlayer_2
   WavePlayer_3
   WavePlayer_4
   WavePlayer_5
   WavePlayer_6
   WavePlayer_7
   WavePlayer_8
   WavePlayer_9
   WavePlayer_10
   WavePlayer_11
   WavePlayer_21
   null
   null
   null
   null
   null
   null
   null
   null
   null
   null
   null
   null
  ]
  properties {weights [1,1,1,1,1,1,1,1,1,1,1,1,1,1,1,1,1,1,1,1,1,1,1,1,1]}
 }
 {
  id=Random_5
  children [WavePlayer_2, WavePlayer_11, WavePlayer_7, null, null, null, null]
  properties {bShouldExcludeFromBranchCulling=true, weights [1,1,1,1,1,1,1]}
 }
 {id=WavePlayer_0, properties {soundWaveAssetPtr=&/Apollo_HeavyLightning_Close_05}}
 {id=WavePlayer_1, properties {soundWaveAssetPtr=&/Apollo_HeavyLightning_Close_04}}
 {id=WavePlayer_10, properties {soundWaveAssetPtr=&/Apollo_HeavyLightning_Close_11}}
 {id=WavePlayer_11, properties {soundWaveAssetPtr=&/Apollo_HeavyLightning_Close_10}}
 {id=WavePlayer_12, properties {soundWaveAssetPtr=&/Apollo_Lightning_Close_09}}
 {id=WavePlayer_13, properties {soundWaveAssetPtr=&/Apollo_Lightning_Close_10}}
 {id=WavePlayer_14, properties {soundWaveAssetPtr=&/Apollo_Lightning_Close_01}}
 {id=WavePlayer_15, properties {soundWaveAssetPtr=&/Apollo_Lightning_Close_02}}
 {id=WavePlayer_16, properties {soundWaveAssetPtr=&/Apollo_Lightning_Close_03}}
 {id=WavePlayer_17, properties {soundWaveAssetPtr=&/Apollo_Lightning_Close_04}}
 {id=WavePlayer_18, properties {soundWaveAssetPtr=&/Apollo_Lightning_Close_05}}
 {id=WavePlayer_19, properties {soundWaveAssetPtr=&/Apollo_Lightning_Close_06}}
 {id=WavePlayer_2, properties {soundWaveAssetPtr=&/Apollo_HeavyLightning_Close_03}}
 {id=WavePlayer_20, properties {soundWaveAssetPtr=&/Apollo_Lightning_Close_07}}
 {id=WavePlayer_21, properties {soundWaveAssetPtr=&/Apollo_HeavyLightning_Close_09}}
 {id=WavePlayer_23, properties {soundWaveAssetPtr=&/Apollo_Lightning_Close_08}}
 {id=WavePlayer_25, properties {soundWaveAssetPtr=&/Apollo_Distant_Thunder_09}}
 {id=WavePlayer_26, properties {soundWaveAssetPtr=&/Apollo_Distant_Thunder_10}}
 {id=WavePlayer_28, properties {soundWaveAssetPtr=&/Apollo_Distant_Thunder_01}}
 {id=WavePlayer_29, properties {soundWaveAssetPtr=&/Apollo_Distant_Thunder_02}}
 {id=WavePlayer_3, properties {soundWaveAssetPtr=&/Apollo_HeavyLightning_Close_02}}
 {id=WavePlayer_30, properties {soundWaveAssetPtr=&/Apollo_Distant_Thunder_03}}
 {id=WavePlayer_31, properties {soundWaveAssetPtr=&/Apollo_Distant_Thunder_04}}
 {id=WavePlayer_32, properties {soundWaveAssetPtr=&/Apollo_Distant_Thunder_05}}
 {id=WavePlayer_33, properties {soundWaveAssetPtr=&/Apollo_Distant_Thunder_06}}
 {id=WavePlayer_34, properties {soundWaveAssetPtr=&/Apollo_Distant_Thunder_07}}
 {id=WavePlayer_35, properties {soundWaveAssetPtr=&/Apollo_Distant_Thunder_08}}
 {id=WavePlayer_4, properties {soundWaveAssetPtr=&/Apollo_HeavyLightning_Close_01}}
 {id=WavePlayer_5, properties {soundWaveAssetPtr=&/Apollo_HeavyLightning_Close_06}}
 {id=WavePlayer_6, properties {soundWaveAssetPtr=&/Apollo_HeavyLightning_Close_07}}
 {id=WavePlayer_7, properties {soundWaveAssetPtr=&/Apollo_HeavyLightning_Close_08}}
 {id=WavePlayer_8, properties {soundWaveAssetPtr=&/Apollo_HeavyLightning_Close_13}}
 {id=WavePlayer_9, properties {soundWaveAssetPtr=&/Apollo_HeavyLightning_Close_12}}
]
properties {
 attenuationSettings=/Game/Sounds/Attenuation/Fort_SafeZone_Lightning_Athena_Att
 maxDistance=2097152
 soundClassObject=/Game/Sounds/FortSoundClassesAndModes/Fort_Storms_PSM
 volumeMultiplier=1.002
 preEffectBusSends [{}]
}
