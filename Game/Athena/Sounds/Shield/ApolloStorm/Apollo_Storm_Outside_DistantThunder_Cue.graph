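# Sound cue for distant thunder heard outside the Apollo storm in Athena.
# A quality-level switch selects a ten-variation random pool, a two-variation random pool, or one of two single waves, all using equal weights where randomized.
# The smaller branches reuse waves from the full pool.
# A distant safe-zone thunder attenuation and storm sound class apply, and a concurrency override caps simultaneous plays, blocking new ones at the limit.
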SoundCue /Game/Athena/Sounds/Shield/ApolloStorm/Apollo_Storm_Outside_DistantThunder_Cue
root=QualityLevel_0
nodes [
 {id=QualityLevel_0, children [Random_3, Random_0, WavePlayer_25, WavePlayer_26]}
 {id=Random_0, children [WavePlayer_25, WavePlayer_26], properties {weights [1,1]}}
 {
  id=Random_3
  children [
   WavePlayer_25
   WavePlayer_26
   WavePlayer_28
   WavePlayer_29
   WavePlayer_30
   WavePlayer_31
   WavePlayer_32
   WavePlayer_33
   WavePlayer_34
   WavePlayer_35
  ]
  properties {weights [1,1,1,1,1,1,1,1,1,1]}
 }
 {id=WavePlayer_25, properties {soundWaveAssetPtr=&/Apollo_Distant_Thunder_09}}
 {id=WavePlayer_26, properties {soundWaveAssetPtr=&/Apollo_Distant_Thunder_10}}
 {id=WavePlayer_28, properties {soundWaveAssetPtr=&/Apollo_Distant_Thunder_01}}
 {id=WavePlayer_29, properties {soundWaveAssetPtr=&/Apollo_Distant_Thunder_02}}
 {id=WavePlayer_30, properties {soundWaveAssetPtr=&/Apollo_Distant_Thunder_03}}
 {id=WavePlayer_31, properties {soundWaveAssetPtr=&/Apollo_Distant_Thunder_04}}
 {id=WavePlayer_32, properties {soundWaveAssetPtr=&/Apollo_Distant_Thunder_05}}
 {id=WavePlayer_33, properties {soundWaveAssetPtr=&/Apollo_Distant_Thunder_06}}
 {id=WavePlayer_34, properties {soundWaveAssetPtr=&/Apollo_Distant_Thunder_07}}
 {id=WavePlayer_35, properties {soundWaveAssetPtr=&/Apollo_Distant_Thunder_08}}
]
properties {
 attenuationSettings=/Game/Sounds/Attenuation/Fort_SafeZone_Thunder_Distant_Athena_Att
 bOverrideConcurrency=true
 maxDistance=2097152
 soundClassObject=/Game/Sounds/FortSoundClassesAndModes/Fort_Storms_PSM
 volumeMultiplier=1.002
 concurrencyOverrides {maxCount=3, resolutionRule=PreventNew}
 preEffectBusSends [{}]
}
